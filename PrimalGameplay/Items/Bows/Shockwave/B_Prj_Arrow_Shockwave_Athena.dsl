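; Shockwave bow arrow projectile that turns a fully charged impact into a shockwave grenade effect.
; On the server, it spawns an invisible, motionless dummy grenade at the hit location, passing along the explode effect container, and attaches it to the hit actor.
; The grenade receives the hit actor and arrow forward vector for special attachment handling.
; A separate pawn-hit function, uncalled here, sets the attached pawn and launch offset.
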(event EventBeginPlay
  (|CallParentFunction) ; This node is disabled and will not be called.\nDrag off pins to build functionality.
)

(event Collision|EventActorBeginOverlap
  (params (OtherActor :type Actor%20Object%20Reference))
  (|CallParentFunction OtherActor) ; This node is disabled and will not be called.\nDrag off pins to build functionality.
)

(event EventTick
  (params (DeltaSeconds :type Float%20%28single-precision%29))
  (|CallParentFunction DeltaSeconds) ; This node is disabled and will not be called.\nDrag off pins to build functionality.
)

(event EventOnFullyChargedImpact
  (params (Hit%20Result :type Hit%20Result%20Structure))
  (if (Networking|IsServer)
    (CallFunction|SpawnShockwaveGrenade Hit%20Result)))

(event Custom|SpawnShockwaveGrenade
  (params (Hit :type Hit%20Result%20Structure))
  (bind (_bblockinghit _binitialoverlap _time _distance _location) (Collision|BreakHitResult Hit))
  (bind (_bblockinghit_2 _binitialoverlap_2 _time_2 _distance_2 _location_2 _impactpoint _normal _impactnormal _physmat _hitactor) (Collision|BreakHitResult Hit))
  ;[1 1 1] Spawn an invisible shockwave grenade and force it to stop
  (bind _returnvalue (Projectile|SpawnProjectile :ProjectileClass (Variables|Default|GetDummyGrenadeClass) :RequestedBy (Game|GetInstigator) :SpawnLocation _location :SpawnRotation "0, 0, 0" :SpawnDirection (Variables|Default|GetZeroRotator) :EffectContainerSpecToApplyOnExplode (Variables|Delivery|GetEffectContainerSpecToApplyOnExplode) :bUseDefaultPhysics false :InitialSpeed 0.0 :GravityScale 0.0 :ChargePercent 0.0))
  (Game|Components|ProjectileMovement|StopSimulating (Variables|/Script/FortniteGame.FortProjectileBase|GetProjectileMovementComponent _returnvalue) Hit)
  ;[1 1 1] Attach the grenade to the actor we hit
  (Transformation|AttachActorToActor _returnvalue _hitactor "None" "KeepWorld" "KeepWorld" "KeepWorld" true)
  ;[1 1 1] Tell the grenade what actor it was attached to so it can handle special cases
  (bind _asb_prj_bow_shockwave_dummy_grenade (Utilities|Casting|CastToB_Prj_Bow_Shockwave_DummyGrenade _returnvalue))
  (Class|BPrjBowShockwaveDummyGrenade|CheckSpecialAttachment _asb_prj_bow_shockwave_dummy_grenade _hitactor (Transformation|GetActorForwardVector)))

(fn HandleHitPawn
  (params (Pawn :type Player%20Pawn%20Athena%20Object%20Reference) (Projectile :type Fort%20Projectile%20Base%20Object%20Reference))
  (bind _asb_prj_bow_shockwave_dummy_grenade (Utilities|Casting|CastToB_Prj_Bow_Shockwave_DummyGrenade Projectile))
  (Class|BPrjBowShockwaveDummyGrenade|SetAttachedtoPawn _asb_prj_bow_shockwave_dummy_grenade Pawn :!class /PrimalGameplay/Items/Bows/Shockwave/B_Prj_Bow_Shockwave_DummyGrenade.B_Prj_Bow_Shockwave_DummyGrenade_C)
  (Class|BPrjBowShockwaveDummyGrenade|SetLaunchOffsetfromArrowDirection _asb_prj_bow_shockwave_dummy_grenade (Transformation|GetActorForwardVector)))

(fn ConstructionScript
  (|CallParentFunction))
