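; Dummy grenade projectile for the Shockwave Bow that plays a timer beep on spawn and adjusts the shockwave launch origin according to what the arrow is attached to.
; CheckSpecialAttachment stores an offset opposite the arrow direction, then tags the attached actor as a player pawn or physics object.
; GetLaunchOriginPoint switches on that tag, offsetting from the pawn mesh or the hit location (with extra Z for physics objects), otherwise using the parent's origin.
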(event EventBeginPlay
  (|CallParentFunction)
  (Audio|Components|Audio|Play (Variables|Default|GetBeepTimer) 0.0))

(fn CheckSpecialAttachment
  (params (AttachedActor :type Actor%20Object%20Reference) (ArrowForwardVector :type Vector))
  (CallFunction|SetLaunchOffsetfromArrowDirection ArrowForwardVector)
  (bind _output_get (Variables|Default|SetAttachedActor AttachedActor))
  (Utilities|Casting|CastToPlayerPawn_Athena _output_get
    (:then
      (Variables|Default|SetSpecialAttachmentType "NewEnumerator0")
      (Development|PrintString "Shockwave arrow attached to pawn"))
    (:CastFailed
      (Utilities|IsValid (Actor|GetComponentbyClass _output_get "/Script/FortniteGame.FortPhysicsObjectComponent")
        (:Is%20Valid
          (Variables|Default|SetSpecialAttachmentType "NewEnumerator1")
          (Development|PrintString "Shockwave arrow attached to physics object"))))))

(fn GetLaunchOriginForPawn
  (bind _asplayer_pawn_athena (Utilities|Casting|CastToPlayerPawn_Athena (Variables|Default|GetAttachedActor)))
  (return (+ (Transformation|GetSocketLocation (Variables|/Script/Engine.Character|GetMesh _asplayer_pawn_athena) "None") (Variables|Default|GetLaunchOffset))))

(fn GetLaunchOriginForPhysicsObject
  (return (Math|Vector|vector+vector (CallFunction|GetOffsetPoint (Variables|BPrjGrenade|GetHitLocation)) (Math|Vector|MakeVector 0.0 0.0 (Variables|Default|GetPhysicsObjectLaunchZOffset)))))

(fn GetLaunchOriginPoint
  ;[1 1 1] Check if we are attached to an actor that calls for a special case and return the launch origin based on that case
  (switch Utilities|FlowControl|Switch|SwitchonEShockwaveBowSpecialAttachmentType (Variables|Default|GetSpecialAttachmentType)
    (:NewEnumerator2
      (return (|CallParentFunction)))
    (:NewEnumerator0
      (return (CallFunction|GetLaunchOriginforPawn)))
    (:NewEnumerator1
      (return (CallFunction|GetLaunchOriginforPhysicsObject)))))

(fn GetOffsetPoint
  (params (OriginalLocation :type Vector))
  (return (+ OriginalLocation (Variables|Default|GetLaunchOffset))))

(fn SetAttachedToPawn
  (params (Pawn :type Player%20Pawn%20Athena%20Object%20Reference))
  (Variables|Default|SetAttachedPawn Pawn)
  (Variables|Default|SetbAttachedToPawn true))

(fn SetLaunchOffsetFromArrowDirection
  (params (ArrowVelocity :type Vector))
  (Variables|Default|SetLaunchOffset (* (Math|Vector|Normalize ArrowVelocity) (* (Variables|Default|GetLaunchOffsetMagnitude) -1.0))))

(fn ConstructionScript
  (|CallParentFunction))
