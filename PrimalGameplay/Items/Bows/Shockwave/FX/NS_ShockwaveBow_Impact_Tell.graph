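# Niagara system for the Shockwave bow's impact tell, a brief purple energy flash that, by its name, telegraphs the shockwave.
# Arcs spawns short-lived rotating flipbook sprites, scaled over both particle age and emitter loop age.
# Fire bursts two flipbook sprites from random start frames that swell, then shrink.
# Light drives a light and a pulsing mesh from one particle, with alpha fading on emitter loop age.
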NiagaraSystem /PrimalGameplay/Items/Bows/Shockwave/FX/NS_ShockwaveBow_Impact_Tell
dependencies {
 usedDataInterfaces [&NiagaraDataInterfaceCurve, &NiagaraDataInterfaceColorCurve]
 usedDynamicInputs []
 usedModules [&SystemState]
}
emitters {
 Arcs {
  emitterUpdateScript [
   &EmitterState(Life Cycle Mode=::NewEnumerator1, Loop Behavior=::NewEnumerator1, Loop Duration=0.225, UseLoopDelay=false)
   &SpawnRate(SpawnRate=12, Use Spawn Probability=false)
  ]
  particleSpawnScript [
   /Niagara/Modules/Spawn/Initialization/InitializeParticle(
    Color=<rgba 100 100 100 1>
    Lifetime=&RandomRangeFloat(Minimum=0.133, Maximum=0.166)
    Module.Write Color=true
    Module.Write Lifetime=true
    Module.Write Mass=true
    Module.Write Material Random=true
    Module.Write Position=false
    Module.Write Scale=false
    Module.Write SpriteRotation=true
    Module.Write SpriteSize=true
    Position=&SimulationPosition()
    Sprite Rotation=&RandomRangeFloat(Minimum=-180, Maximum=180)
    Sprite Size=&Vector2DFromFloat(Value=&RandomRangeFloat(Minimum=100, Maximum=120))
   )
   &SubUVAnimation(End Frame=3, SubUV Animation Mode:ENiagara_SubUVLookupMode=::NewEnumerator2)
  ]
  particleUpdateScript [
   &ParticleState()
   &CameraOffset(Camera Offset Amount=30)
   &ScaleSpriteSize(Scale Factor=&Vector2DFromFloat(Value=&FloatFromCurve(FloatCurve=DataInterface() {propertyValues {Curve {keys [
          {arriveTangent=1.3754, interpMode=RCIM_Cubic, leaveTangent=1.3754, tangentMode=RCTM_User, time=0.0023, value=0.1}
          {arriveTangent=0.2702, interpMode=RCIM_Cubic, leaveTangent=0.2702, tangentMode=RCTM_User, time=0.66, value=1}
          {arriveTangent=-2.8829, interpMode=RCIM_Cubic, leaveTangent=-2.8829, tangentMode=RCTM_User, time=1, value=0.33}
         ]}}})))
   &ScaleSpriteSize(Scale Factor=&Vector2DFromFloat(Value=&FloatFromCurve(
      FloatCurve=DataInterface() {propertyValues {Curve {keys [
          {arriveTangent=2.1557, interpMode=RCIM_Cubic, leaveTangent=0.6478, tangentMode=RCTM_User, time=0, value=1}
          <at 0.66 1.5 RCIM_Cubic>
          {arriveTangent=-1.0096, interpMode=RCIM_Cubic, leaveTangent=-0.4296, tangentMode=RCTM_User, time=1, value=1}
         ]}}}
      CurveIndex=$Emitter.NormalizedLoopAge
     ))) {module=ScaleSpriteSize001}
   &Color(Color=&ColorFromCurve(ColorCurve=DataInterface() {propertyValues {
       AlphaCurve {keys [<at 0.4996 1 RCIM_Cubic>,<at 1.0091 0 RCIM_Cubic>]}
       BlueCurve {keys [<at 0.5015 1 RCIM_Cubic>,<at 1 1 RCIM_Cubic>]}
       GreenCurve {keys [<at 0.5015 0.2242 RCIM_Cubic>,<at 1 0.1557 RCIM_Cubic>]}
       RedCurve {keys [<at 0.5015 0.5656 RCIM_Cubic>,<at 1 0.1073 RCIM_Cubic>]}
      }}))
   &SpriteRotationRate(Rotation Rate=100)
  ]
  rendererClasses [&NiagaraSpriteRendererProperties]
  renderers [{rendererClass=&NiagaraSpriteRendererProperties}]
 }
 Fire {
  emitterUpdateScript [
   &EmitterState()
   &SpawnBurst_Instantaneous(Spawn Count=2, Use Spawn Probability=false)
  ]
  particleSpawnScript [
   &InitializeParticle(
    Color=<rgba 6.001 2.6654 10 1>
    Lifetime=0.4
    Module.Write Color=true
    Module.Write Lifetime=true
    Module.Write Mass=true
    Module.Write Material Random=true
    Module.Write Position=false
    Module.Write Scale=false
    Module.Write SpriteRotation=true
    Module.Write SpriteSize=true
    Position=&SimulationPosition()
    Sprite Rotation=&RandomRangeFloat(Minimum=-180, Maximum=180)
    Sprite Size=<40 40>
   )
   /SetVariables(Particles.RandomFrame=&RandomRangeInt(Minimum=0, Maximum=63)) {isSetParameters=true}
  ]
  particleUpdateScript [
   &ParticleState()
   &CameraOffset(Camera Offset Amount=50)
   &FN_SubUVAnimation(End Frame=&Add_Integer(A=$Particles.RandomFrame, B=63), Start Frame=$Particles.RandomFrame)
   &ScaleSpriteSize(Scale Factor=&Vector2DFromFloat(Value=&FloatFromCurve(FloatCurve=DataInterface() {propertyValues {Curve {keys [
          {arriveTangent=2.5076, interpMode=RCIM_Cubic, leaveTangent=2.5076, tangentMode=RCTM_User, time=0.0023, value=0}
          {arriveTangent=0.0897, interpMode=RCIM_Cubic, leaveTangent=0.0897, tangentMode=RCTM_User, time=1, value=1}
         ]}}}))) {enabled=false}
   &ScaleSpriteSize(Scale Factor=&Vector2DFromFloat(Value=&FloatFromCurve(FloatCurve=DataInterface() {propertyValues {Curve {keys [
          {arriveTangent=5.4036, interpMode=RCIM_Cubic, leaveTangent=4.7745, tangentMode=RCTM_User, time=0, value=0.1}
          <at 0.66 2 RCIM_Cubic>
          {arriveTangent=-10.705, interpMode=RCIM_Cubic, leaveTangent=-33.5072, tangentMode=RCTM_User, time=1, value=0.1}
         ]}}}))) {module=ScaleSpriteSize001}
  ]
  rendererClasses [&NiagaraSpriteRendererProperties]
  renderers [{rendererClass=&NiagaraSpriteRendererProperties}]
 }
 Light {
  emitterUpdateScript [
   &EmitterState(Inactive Response=::NewEnumerator1, Life Cycle Mode=::NewEnumerator1, Loop Behavior=::NewEnumerator1, Loop Duration=0.25, UseLoopDelay=false)
   &SpawnBurst_Instantaneous(Spawn Count=1, Use Spawn Probability=false)
  ]
  particleSpawnScript [&InitializeParticle(
    Color=<rgba 3.0005 1.3327 5 0.05>
    Lifetime=0.2
    Module.Write Color=true
    Module.Write Lifetime=true
    Module.Write Mass=true
    Module.Write Material Random=true
    Module.Write Position=false
    Module.Write Scale=false
    Module.Write SpriteRotation=false
    Module.Write SpriteSize=true
    Position=&SimulationPosition()
    Sprite Size=<80 80>
   )]
  particleUpdateScript [
   &ParticleState(Kill Particles When Lifetime Has Elapsed=false, Let Infinitely Lived Particles Die When Emitter Deactivates=true)
   &CameraOffset(Camera Offset Amount=30)
   &ScaleMeshSize(Scale Factor=&VectorFromFloat(Value=&FloatFromCurve(FloatCurve=DataInterface() {propertyValues {Curve {keys [
          {arriveTangent=0.5924, interpMode=RCIM_Cubic, leaveTangent=-0.0686, tangentMode=RCTM_User, time=0, value=0.2}
          <at 0.66 1 RCIM_Cubic>
          {arriveTangent=-3.1773, interpMode=RCIM_Cubic, leaveTangent=-12.5808, tangentMode=RCTM_User, time=1, value=0.2}
         ]}}})))
   &ScaleSpriteSize(Scale Factor=&Vector2DFromFloat(Value=&FloatFromCurve(
      FloatCurve=DataInterface() {propertyValues {Curve {keys [
          {arriveTangent=0.5924, interpMode=RCIM_Cubic, leaveTangent=-0.0686, tangentMode=RCTM_User, time=0, value=0.1}
          <at 0.66 2 RCIM_Cubic>
          {interpMode=RCIM_Cubic, leaveTangent=-5.6923, time=1, value=0.1}
         ]}}}
      CurveIndex=$Emitter.NormalizedLoopAge
     ))) {enabled=false, module=ScaleSpriteSize001}
   &ScaleColor(
    Scale Alpha=&FloatFromCurve(
     FloatCurve=DataInterface() {propertyValues {Curve {keys [
         {arriveTangent=9.0751, interpMode=RCIM_Cubic, time=0, value=0}
         <at 0.2846 1 RCIM_Cubic>
         {interpMode=RCIM_Cubic, leaveTangent=-1.1063, time=1, value=0}
        ]}}}
     CurveIndex=$Emitter.NormalizedLoopAge
    )
    ScaleA=true
    ScaleRGB=true
   )
   &FN_Light_v3(
    Color=<rgba 7.1849 1.9607 15 1>
    Create Color Attribute=true
    Create Exponent Attribute=true
    Create Radius Attribute=true
    Exponent=5
    Radius=150
   )
  ]
  rendererClasses [&NiagaraLightRendererProperties, &NiagaraMeshRendererProperties]
  renderers [
   {rendererClass=&NiagaraLightRendererProperties, rendererIndex=0}
   {rendererClass=&NiagaraMeshRendererProperties, rendererIndex=1}
  ]
 }
}
summary {
 systemName=NS_ShockwaveBow_Impact_Tell
 emitters [
  {emitterName=Light, rendererClasses [&NiagaraLightRendererProperties, &NiagaraMeshRendererProperties]}
  {emitterName=Fire, rendererClasses [&NiagaraSpriteRendererProperties]}
  {emitterName=Arcs, rendererClasses [&NiagaraSpriteRendererProperties]}
 ]
 userVariables [User.Offset {type=Vector3f, description="", defaultValue {struct=Vector3f, value=<8 0 0>}}]
}
